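# M_SoftGlowOutlines is an unlit, translucent, two-sided material for soft glowing outlines in the Material Gym.
# The graph is a single call to the MF_SoftGlowOutlines material function, with material attributes enabled, so the effect's logic lives in that function.
# It is flagged for use on skeletal meshes and with mesh deformers.
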Material /Game/Global/Materials/MaterialGym/Materials/SoftGlowOutlines/M_SoftGlowOutlines
graph {
 properties {
  bUseMaterialAttributes=true
  bUsedWithMeshDeformer=true
  bUsedWithSkeletalMesh=true
  blendMode=BLEND_Translucent
  pixelDepthOffsetMode=PDOM_Legacy
  shadingModel=MSM_Unlit
  twoSided=true
  usedShadingModels=Unlit
 }
 nodes [MaterialFunctionCall() {materialFunction=&/MF_SoftGlowOutlines}]
}
